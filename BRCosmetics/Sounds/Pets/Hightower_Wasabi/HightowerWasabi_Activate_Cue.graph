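# Sound cue for the activation sound of the Hightower Wasabi pet cosmetic.
# A single wave player plays the HightowerWasabi_Activate wave at reduced volume, with no randomization or mixing.
# It routes through the gadgets and items sound class and uses the close-range pet attenuation, so it is heard only near the pet.
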SoundCue /BRCosmetics/Sounds/Pets/Hightower_Wasabi/HightowerWasabi_Activate_Cue
root=WavePlayer_0
nodes [{id=WavePlayer_0, properties {soundWaveAssetPtr=&/HightowerWasabi_Activate}}]
properties {
 attenuationSettings=&Fort_Pets_Close_Athena_Att
 maxDistance=1800.0001
 soundClassObject=&Fort_Gadgets_Items
 volumeMultiplier=0.5
}
