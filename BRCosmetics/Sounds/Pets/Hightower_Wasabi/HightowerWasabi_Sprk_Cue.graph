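# Sound cue for a spark effect ("Sprk") belonging to the Hightower Wasabi pet cosmetic.
# A Random node picks one of four equally weighted spark wave variations and feeds it into a Modulator that randomizes pitch with a raised upper limit for extra variety.
# The cue plays at a heavily reduced volume through the gadgets and items sound class.
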SoundCue /BRCosmetics/Sounds/Pets/Hightower_Wasabi/HightowerWasabi_Sprk_Cue
root=Modulator_0
nodes [
 {id=Modulator_0, children [Random_0], properties {pitchMax=1.5}}
 {id=Random_0, children [WavePlayer_1,WavePlayer_2,WavePlayer_3,WavePlayer_4], properties {weights [1,1,1,1]}}
 {id=WavePlayer_1, properties {soundWaveAssetPtr=&/HightowerWasabi_Sprk_01}}
 {id=WavePlayer_2, properties {soundWaveAssetPtr=&/HightowerWasabi_Sprk_02}}
 {id=WavePlayer_3, properties {soundWaveAssetPtr=&/HightowerWasabi_Sprk_03}}
 {id=WavePlayer_4, properties {soundWaveAssetPtr=&/HightowerWasabi_Sprk_04}}
]
properties {maxDistance=2097152, soundClassObject=&Fort_Gadgets_Items, volumeMultiplier=0.05}
