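# Physics asset for the HighTower Wasabi backpack cosmetic, defining simulation bodies for its dangling cords, fringe and loops.
# A compound pack body of capsules and a box anchors the capsule cord chains and three box-shaped loops.
# Fringe capsules hang from the ends of big cords B, C and D.
# Constraints use tight swing and twist limits, except the cordA chain, whose angular motion is free.
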PhysicsAsset /BRCosmetics/Accessories/FORT_Backpacks/Backpack_M_MED_HighTower_Wasabi/Meshes/M_MED_HighTower_Wasabi_Pack_Physics
bodies {
 dyn_bigCordA_01 [{shapeType=Capsule, length=2.7271, radius=1.5884, center=<2.6961 0 0.0001>, rotation=<rot 90 0 0>}]
 dyn_bigCordA_02 [{shapeType=Capsule, length=0.7628, radius=1.5074, center=<0.8147 0 0>, rotation=<rot 90 0 0>}]
 dyn_bigCordA_03 [{shapeType=Capsule, length=2.9806, radius=0.5, center=<1.4955 0 0>, rotation=<rot 90 0 0>}]
 dyn_bigCordB_01 [{shapeType=Capsule, length=1.8591, radius=1.7288, center=<1.5 0 0>, rotation=<rot 90 0 0>}]
 dyn_bigCordB_02 [{shapeType=Capsule, length=0.6076, radius=1.5811, center=<0.4836 0 0>, rotation=<rot 90 0 0>}]
 dyn_bigCordC_01 [{shapeType=Capsule, length=1.5244, radius=1.491, center=<1.5 0 0>, rotation=<rot 90 0 0>}]
 dyn_bigCordC_02 [{shapeType=Capsule, length=0.3777, radius=1.5446, center=<0.8166 0 0>, rotation=<rot 90 0 0>}]
 dyn_bigCordD_01 [{shapeType=Capsule, length=2.9366, radius=1.6247, center=<2.3132 0 0>, rotation=<rot 90 0 0>}]
 dyn_bigCordD_02 [{shapeType=Capsule, length=0.2229, radius=1.5561, center=<0.7527 0 0>, rotation=<rot 90 0 0>}]
 dyn_cordA_01 [{shapeType=Capsule, length=2.3426, radius=0.5191, center=<1 0 0>, rotation=<rot 90 0 0>}]
 dyn_cordA_02 [{shapeType=Capsule, length=2.3426, radius=0.53, center=<1 0 0>, rotation=<rot 90 0 0>}]
 dyn_cordB_01 [{shapeType=Capsule, length=2.5534, radius=0.6345, center=<1.5681 0 0>, rotation=<rot 90 0 0>}]
 dyn_cordB_02 [{shapeType=Capsule, length=2.6714, radius=0.6248, center=<1.5608 0.0001 0>, rotation=<rot 90 0 0>}]
 dyn_cordB_03 [{shapeType=Capsule, length=2.4927, radius=0.7333, center=<1.3325 0 0>, rotation=<rot 90 0 0>}]
 dyn_cordC_01 [{shapeType=Capsule, length=4.4525, radius=0.5645, center=<2.4313 0 0>, rotation=<rot 90 0 0>}]
 dyn_cordD_01 [{shapeType=Capsule, length=2.3426, radius=0.505, center=<1.5681 0 0>, rotation=<rot 90 0 0>}]
 dyn_cordD_02 [{shapeType=Capsule, length=3.7787, radius=0.5, center=<2.0007 0 0>, rotation=<rot 90 0 0>}]
 dyn_cordE_01 [{shapeType=Capsule, length=2.8641, radius=0.505, center=<1.6435 0 0>, rotation=<rot 90 0 0>}]
 dyn_cordE_02 [{shapeType=Capsule, length=4.6173, radius=0.505, center=<2.4826 0 0>, rotation=<rot 90 0 0>}]
 dyn_cordF_01 [{shapeType=Capsule, length=3.3845, radius=0.5524, center=<1.8106 0 0>, rotation=<rot 90 0 0>}]
 dyn_cordF_02 [{shapeType=Capsule, length=4.8532, radius=0.5951, center=<2.6576 0 0>, rotation=<rot 90 0 0>}]
 dyn_fringeB_A_01 [{shapeType=Capsule, length=1.1314, radius=0.5, center=<1.3956 0.0034 -0.0006>, rotation=<rot 0 75.3635 -87.9292>}]
 dyn_fringeB_B_01 [{shapeType=Capsule, length=1.1314, radius=0.5, center=<1.7931 -0.0485 -0.0455>, rotation=<rot 0 94.2723 -87.5246>}]
 dyn_fringeB_C_01 [{shapeType=Capsule, length=1.1314, radius=0.5, center=<1.306 -0.0186 0.01>, rotation=<rot 0 -82.659 -80.2511>}]
 dyn_fringeB_D_01 [{shapeType=Capsule, length=1.1314, radius=0.5, center=<1.3747 0.0238 0.0297>, rotation=<rot 0 -98.7304 -84.4457>}]
 dyn_fringeC_A_01 [{shapeType=Capsule, length=0.505, radius=0.505, center=<1.3596 0.0003 -0.0008>, rotation=<rot 0 76.433 -89.7088>}]
 dyn_fringeC_B_01 [{shapeType=Capsule, length=2.0478, radius=0.4349, center=<1.9985 0.0621 0.0781>, rotation=<rot 0 -91.1508 -89.5367>}]
 dyn_fringeC_C_01 [{shapeType=Capsule, length=0.505, radius=0.505, center=<1.7959 0.0115 0.0612>, rotation=<rot 0 -91.9074 -84.4585>}]
 dyn_fringeC_D_01 [{shapeType=Capsule, length=1.3963, radius=0.4047, center=<1.3537 0.035 0.0095>, rotation=<rot 0 88.2366 -89.7661>}]
 dyn_fringeD_A_01 [{shapeType=Capsule, length=0.9407, radius=0.5, center=<2.326 0.0457 -0.0059>, rotation=<rot 0 82.1982 -89.5176>}]
 dyn_fringeD_B_01 [{shapeType=Capsule, length=1.3433, radius=0.5, center=<2.0087 0.057 0.0355>, rotation=<rot 0 -95.1618 -87.7197>}]
 dyn_fringeD_C_01 [{shapeType=Capsule, length=0.505, radius=0.505, center=<1.1256 -0.0094 0.0132>, rotation=<rot 0 -92.7434 -86.8688>}]
 dyn_fringeD_C_02 [{shapeType=Capsule, length=0.505, radius=0.505, center=<1.231 -0.0033 -0.0166>, rotation=<rot 0 -86.4394 -78.1068>}]
 dyn_loopA [{shapeType=Box, extentX=1.2254, extentY=4.0653, extentZ=5.0166, center=<2.2419 0.0007 -0.3212>, rotation=<rot -4.6322 -90 9.4941>}]
 dyn_loopB [{shapeType=Box, extentX=1.2657, extentY=5.7394, extentZ=8.3722, center=<3.2681 0.0431 0.8423>, rotation=<rot -0.4066 -90 -19.9987>}]
 dyn_loopC [{shapeType=Box, extentX=1.2206, extentY=5.2849, extentZ=7.7093, center=<3.867 -0.0606 0.6235>, rotation=<rot -0.0325 -90 -12.9339>}]
 pack [
  {shapeType=Capsule, length=15.3633, radius=1.6023, center=<0.0349 -2.4943 -1.3124>, rotation=<rot -18 -90 0.6727>}
  {shapeType=Capsule, length=25.2554, radius=34.7337, center=<-0.0008 34.7217 0.516>, rotation=<rot 0 0 -6.9437>}
  {shapeType=Capsule, length=26.1495, radius=15, center=<12.8986 14.9392 0>, rotation=<rot 0 0 -3.4522>}
  {shapeType=Capsule, length=26.1495, radius=15, center=<-13.4297 14.9391 0.0003>, rotation=<rot 0 0 -3.4522>}
  {shapeType=Capsule, length=15, radius=15, center=<0 -20.8189 -0.9006>, rotation=<rot 90 0 0>}
  {shapeType=Box, extentX=40.569, extentY=4.1382, extentZ=23.1611, center=<-0.0001 2.4164 -14.175>, rotation=<rot 0 0 -2.6988>}
 ]
}
constraints [
 {bone1Name=dyn_cordF_01, bone2Name=pack, swing1LimitDegrees=16, swing2LimitDegrees=16, twistLimitDegrees=8}
 {bone1Name=dyn_cordF_02, bone2Name=dyn_cordF_01, swing1LimitDegrees=16, swing2LimitDegrees=16, twistLimitDegrees=8}
 {bone1Name=dyn_cordE_01, bone2Name=pack, swing1LimitDegrees=16, swing2LimitDegrees=16, twistLimitDegrees=8}
 {bone1Name=dyn_cordE_02, bone2Name=dyn_cordE_01, swing1LimitDegrees=16, swing2LimitDegrees=16, twistLimitDegrees=8}
 {bone1Name=dyn_cordD_01, bone2Name=pack, swing1LimitDegrees=16, swing2LimitDegrees=16, twistLimitDegrees=8}
 {bone1Name=dyn_cordD_02, bone2Name=dyn_cordD_01, swing1LimitDegrees=16, swing2LimitDegrees=16, twistLimitDegrees=8}
 {bone1Name=dyn_cordC_01, bone2Name=pack, swing1LimitDegrees=16, swing2LimitDegrees=16, twistLimitDegrees=8}
 {bone1Name=dyn_cordB_01, bone2Name=pack, swing1LimitDegrees=16, swing2LimitDegrees=16, twistLimitDegrees=8}
 {bone1Name=dyn_cordB_02, bone2Name=dyn_cordB_01, swing1LimitDegrees=16, swing2LimitDegrees=16, twistLimitDegrees=8}
 {bone1Name=dyn_cordB_03, bone2Name=dyn_cordB_02, swing1LimitDegrees=16, swing2LimitDegrees=16, twistLimitDegrees=8}
 {
  bone1Name=dyn_cordA_01
  bone2Name=pack
  swing1LimitDegrees=45
  swing1Motion=Free
  swing2LimitDegrees=45
  swing2Motion=Free
  twistLimitDegrees=45
  twistMotion=Free
 }
 {
  bone1Name=dyn_cordA_02
  bone2Name=dyn_cordA_01
  swing1LimitDegrees=45
  swing1Motion=Free
  swing2LimitDegrees=45
  swing2Motion=Free
  twistLimitDegrees=45
  twistMotion=Free
 }
 {bone1Name=dyn_bigCordB_01, bone2Name=pack, swing1LimitDegrees=12, swing2LimitDegrees=12, twistLimitDegrees=4}
 {bone1Name=dyn_bigCordB_02, bone2Name=dyn_bigCordB_01, swing1LimitDegrees=12, swing2LimitDegrees=12, twistLimitDegrees=4}
 {bone1Name=dyn_bigCordC_01, bone2Name=pack, swing1LimitDegrees=14, swing2LimitDegrees=14, twistLimitDegrees=4}
 {bone1Name=dyn_bigCordC_02, bone2Name=dyn_bigCordC_01, swing1LimitDegrees=14, swing2LimitDegrees=14, twistLimitDegrees=4}
 {bone1Name=dyn_bigCordD_01, bone2Name=pack, swing1LimitDegrees=12, swing2LimitDegrees=12, twistLimitDegrees=4}
 {bone1Name=dyn_bigCordD_02, bone2Name=dyn_bigCordD_01, swing1LimitDegrees=12, swing2LimitDegrees=12, twistLimitDegrees=4}
 {bone1Name=dyn_fringeB_C_01, bone2Name=dyn_bigCordB_02, swing1LimitDegrees=18, swing2LimitDegrees=18, twistLimitDegrees=6}
 {bone1Name=dyn_fringeB_D_01, bone2Name=dyn_bigCordB_02, swing1LimitDegrees=18, swing2LimitDegrees=18, twistLimitDegrees=6}
 {bone1Name=dyn_fringeB_B_01, bone2Name=dyn_bigCordB_02, swing1LimitDegrees=18, swing2LimitDegrees=18, twistLimitDegrees=6}
 {bone1Name=dyn_fringeB_A_01, bone2Name=dyn_bigCordB_02, swing1LimitDegrees=18, swing2LimitDegrees=18, twistLimitDegrees=6}
 {bone1Name=dyn_fringeC_B_01, bone2Name=dyn_bigCordC_02, swing1LimitDegrees=18, swing2LimitDegrees=18, twistLimitDegrees=8}
 {bone1Name=dyn_fringeC_D_01, bone2Name=dyn_bigCordC_02, swing1LimitDegrees=18, swing2LimitDegrees=18, twistLimitDegrees=8}
 {bone1Name=dyn_fringeC_C_01, bone2Name=dyn_bigCordC_02, swing1LimitDegrees=18, swing2LimitDegrees=18, twistLimitDegrees=8}
 {bone1Name=dyn_fringeC_A_01, bone2Name=dyn_bigCordC_02, swing1LimitDegrees=18, swing2LimitDegrees=18, twistLimitDegrees=8}
 {bone1Name=dyn_fringeD_B_01, bone2Name=dyn_bigCordD_02, swing1LimitDegrees=18, swing2LimitDegrees=18, twistLimitDegrees=8}
 {bone1Name=dyn_fringeD_A_01, bone2Name=dyn_bigCordD_02, swing1LimitDegrees=18, swing2LimitDegrees=18, twistLimitDegrees=8}
 {bone1Name=dyn_fringeD_C_02, bone2Name=dyn_fringeD_C_01, swing1LimitDegrees=18, swing2LimitDegrees=18, twistLimitDegrees=6}
 {bone1Name=dyn_fringeD_C_01, bone2Name=dyn_bigCordD_02, swing1LimitDegrees=18, swing2LimitDegrees=18, twistLimitDegrees=6}
 {bone1Name=dyn_bigCordA_01, bone2Name=pack, swing1LimitDegrees=12, swing2LimitDegrees=12, twistLimitDegrees=4}
 {bone1Name=dyn_bigCordA_02, bone2Name=dyn_bigCordA_01, swing1LimitDegrees=12, swing2LimitDegrees=12, twistLimitDegrees=4}
 {bone1Name=dyn_bigCordA_03, bone2Name=dyn_bigCordA_02, swing1LimitDegrees=16, swing2LimitDegrees=16, twistLimitDegrees=6}
 {bone1Name=dyn_loopA, bone2Name=pack, swing1LimitDegrees=12, swing2LimitDegrees=4, twistLimitDegrees=4}
 {bone1Name=dyn_loopB, bone2Name=pack, swing1LimitDegrees=12, swing2LimitDegrees=4, twistLimitDegrees=4}
 {bone1Name=dyn_loopC, bone2Name=pack, swing1LimitDegrees=12, swing2LimitDegrees=4, twistLimitDegrees=4}
]
